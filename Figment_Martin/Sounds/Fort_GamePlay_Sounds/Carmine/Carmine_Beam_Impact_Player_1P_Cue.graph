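# Sound cue for the first-person Carmine beam impact on a player.
# A quality-level node sends its first two levels to a random pick of seven impact waves and its last two to a matching set of _opt variants.
# It plays without attenuation or spatialization, and overridden concurrency caps simultaneous instances, stopping the oldest.
# It is routed to a third-person ranged weapon fire sound class despite the 1P name.
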SoundCue /Figment_Martin/Sounds/Fort_GamePlay_Sounds/Carmine/Carmine_Beam_Impact_Player_1P_Cue
root=QualityLevel_0
nodes [
 {id=QualityLevel_0, children [Random_0,Random_0,Random_1,Random_1]}
 {
  id=Random_0
  children [WavePlayer_1, WavePlayer_2, WavePlayer_3, WavePlayer_4, WavePlayer_5, WavePlayer_9, WavePlayer_12]
  properties {weights [1,1,1,1,1,1,1]}
 }
 {
  id=Random_1
  children [WavePlayer_0, WavePlayer_6, WavePlayer_7, WavePlayer_8, WavePlayer_10, WavePlayer_11, WavePlayer_13]
  properties {weights [1,1,1,1,1,1,1]}
 }
 {id=WavePlayer_0, properties {soundWaveAssetPtr=&/Carmine_PlayerImpact_01_opt}}
 {id=WavePlayer_1, properties {soundWaveAssetPtr=&/Carmine_PlayerImpact_01}}
 {id=WavePlayer_10, properties {soundWaveAssetPtr=&/Carmine_PlayerImpact_05_opt}}
 {id=WavePlayer_11, properties {soundWaveAssetPtr=&/Carmine_PlayerImpact_06_opt}}
 {id=WavePlayer_12, properties {soundWaveAssetPtr=&/Carmine_PlayerImpact_07}}
 {id=WavePlayer_13, properties {soundWaveAssetPtr=&/Carmine_PlayerImpact_07_opt}}
 {id=WavePlayer_2, properties {soundWaveAssetPtr=&/Carmine_PlayerImpact_02}}
 {id=WavePlayer_3, properties {soundWaveAssetPtr=&/Carmine_PlayerImpact_03}}
 {id=WavePlayer_4, properties {soundWaveAssetPtr=&/Carmine_PlayerImpact_04}}
 {id=WavePlayer_5, properties {soundWaveAssetPtr=&/Carmine_PlayerImpact_05}}
 {id=WavePlayer_6, properties {soundWaveAssetPtr=&/Carmine_PlayerImpact_02_opt}}
 {id=WavePlayer_7, properties {soundWaveAssetPtr=&/Carmine_PlayerImpact_03_opt}}
 {id=WavePlayer_8, properties {soundWaveAssetPtr=&/Carmine_PlayerImpact_04_opt}}
 {id=WavePlayer_9, properties {soundWaveAssetPtr=&/Carmine_PlayerImpact_06}}
]
properties {
 bExcludeFromRandomNodeBranchCulling=true
 bOverrideConcurrency=true
 maxDistance=2097152
 soundClassObject=&Fort_WeaponRanged_Fire_3P_PSM
 volumeMultiplier=0.6
 attenuationOverrides {bAttenuate=false, bSpatialize=false}
 concurrencyOverrides {maxCount=3, resolutionRule=StopOldest, volumeScale=0.75}
}
