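# Sound cue for the third-person Carmine beam weapon firing sound, built on start-loop-stop waves.
# A mixer layers close and distant branches, each randomly picking one of two wave variants through its own attenuation asset and the ranged weapon fire 3P sound class.
# The distant branch is lowered in volume by a modulator.
# A concurrency override allows one instance at a time, stopping the oldest.
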SoundCue /Figment_Martin/Sounds/Fort_GamePlay_Sounds/Carmine/Carmine_Beam_StartLoopStop_Cue
root=Mixer_0
nodes [
 {id=Attenuation_0, children [Random_1], properties {attenuationSettings=/Game/Sounds/Attenuation/Fort_Weapon_Carmine_Close_Athena_Att}}
 {
  id=Attenuation_1
  children [Modulator_0]
  properties {attenuationSettings=/Game/Sounds/Attenuation/Fort_Weapon_Carmine_Far_Athena_Att, attenuationOverrides {bAttenuate=false, bSpatialize=false}}
 }
 {id=Mixer_0, children [SoundClass_0,SoundClass_1], properties {inputVolume [1,1]}}
 {id=Modulator_0, children [Random_0], properties {volumeMax=0.5, volumeMin=0.5}}
 {id=Random_0, children [WavePlayer_2,WavePlayer_4], properties {weights [1,1]}}
 {id=Random_1, children [WavePlayer_0,WavePlayer_7], properties {weights [1,1]}}
 {
  id=SoundClass_0
  children [Attenuation_0]
  properties {soundClassOverride=&Fort_WeaponRanged_Fire_3P_PSM}
 }
 {
  id=SoundClass_1
  children [Attenuation_1]
  properties {soundClassOverride=&Fort_WeaponRanged_Fire_3P_PSM}
 }
 {id=WavePlayer_0, properties {soundWaveAssetPtr=&/Carmine_Beam_StartLoopStop_01}}
 {id=WavePlayer_2, properties {soundWaveAssetPtr=&/Carmine_Beam_StartLoopStop_01_distant}}
 {id=WavePlayer_4, properties {soundWaveAssetPtr=&/Carmine_Beam_StartLoopStop_02_distant}}
 {id=WavePlayer_7, properties {soundWaveAssetPtr=&/Carmine_Beam_StartLoopStop_02}}
]
properties {bOverrideConcurrency=true, maxDistance=25000, volumeMultiplier=1, concurrencyOverrides {maxCount=1, resolutionRule=StopOldest}}
