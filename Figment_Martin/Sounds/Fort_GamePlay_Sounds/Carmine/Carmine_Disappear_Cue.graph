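# Sound cue for the Carmine disappear sound in the Figment_Martin gameplay sound set.
# A single wave player feeds the Carmine_Disappear_01 wave straight to the output, with no randomization or mixing.
# It borrows the Carmine close-range weapon attenuation for spatial falloff and routes through the third-person ranged weapon fire sound class.
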SoundCue /Figment_Martin/Sounds/Fort_GamePlay_Sounds/Carmine/Carmine_Disappear_Cue
root=WavePlayer_0
nodes [{id=WavePlayer_0, properties {soundWaveAssetPtr=&/Carmine_Disappear_01}}]
properties {
 attenuationSettings=/Figment_Martin/Sounds/Attenuation/Fort_Weapon_Carmine_Close_Athena_Att
 maxDistance=4500
 soundClassObject=&Fort_WeaponRanged_Fire_3P_PSM
 volumeMultiplier=1
}
