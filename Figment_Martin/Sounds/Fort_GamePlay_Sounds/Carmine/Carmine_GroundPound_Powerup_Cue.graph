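# Sound cue for Carmine's ground pound powerup, pairing a powerup sound with a dive sound that starts after a short fixed delay.
# A quality switch sends its first two levels to standard waves and its last two to "_opt" variants.
# Each branch mixes close and distant layers through separate non-spatialized attenuations, all under a third-person ranged weapon fire sound class.
# Concurrency allows one instance, stopping the oldest.
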SoundCue /Figment_Martin/Sounds/Fort_GamePlay_Sounds/Carmine/Carmine_GroundPound_Powerup_Cue
root=QualityLevel_0
nodes [
 {
  id=Attenuation_0
  children [Mixer_3]
  properties {attenuationSettings=/Figment_Martin/Sounds/Attenuation/Fort_Weapon_Carmine_Close_Athena_Att, attenuationOverrides {bSpatialize=false}}
 }
 {
  id=Attenuation_1
  children [Mixer_0]
  properties {attenuationSettings=&Fort_Weapon_Carmine_Close_Athena_Att, attenuationOverrides {bSpatialize=false}}
 }
 {
  id=Attenuation_2
  children [Mixer_2]
  properties {attenuationSettings=/Figment_Martin/Sounds/Attenuation/Fort_Weapon_Carmine_Far_Athena_Att, attenuationOverrides {bSpatialize=false}}
 }
 {
  id=Attenuation_3
  children [Mixer_5]
  properties {attenuationSettings=&Fort_Weapon_Carmine_Far_Athena_Att, attenuationOverrides {bSpatialize=false}}
 }
 {id=Delay_0, children [WavePlayer_2], properties {delayMax=0.35, delayMin=0.35}}
 {id=Delay_1, children [WavePlayer_4], properties {delayMax=0.35, delayMin=0.35}}
 {id=Delay_2, children [WavePlayer_10], properties {delayMax=0.35, delayMin=0.35}}
 {id=Delay_3, children [WavePlayer_11], properties {delayMax=0.35, delayMin=0.35}}
 {id=Mixer_0, children [WavePlayer_1,Delay_0], properties {inputVolume [1,1]}}
 {id=Mixer_1, children [SoundClass_0,SoundClass_1], properties {inputVolume [1,1]}}
 {id=Mixer_2, children [WavePlayer_5,Delay_1], properties {inputVolume [1,1]}}
 {id=Mixer_3, children [WavePlayer_8,Delay_2], properties {inputVolume [1,1]}}
 {id=Mixer_4, children [SoundClass_2,SoundClass_3], properties {inputVolume [1,1]}}
 {id=Mixer_5, children [WavePlayer_9,Delay_3], properties {inputVolume [1,1]}}
 {id=QualityLevel_0, children [Mixer_1,Mixer_1,Mixer_4,Mixer_4]}
 {
  id=SoundClass_0
  children [Attenuation_1]
  properties {soundClassOverride=&Fort_WeaponRanged_Fire_3P_PSM}
 }
 {
  id=SoundClass_1
  children [Attenuation_2]
  properties {soundClassOverride=&Fort_WeaponRanged_Fire_3P_PSM}
 }
 {
  id=SoundClass_2
  children [Attenuation_0]
  properties {soundClassOverride=&Fort_WeaponRanged_Fire_3P_PSM}
 }
 {
  id=SoundClass_3
  children [Attenuation_3]
  properties {soundClassOverride=&Fort_WeaponRanged_Fire_3P_PSM}
 }
 {id=WavePlayer_1, properties {soundWaveAssetPtr=&/Carmine_GroundPound_Powerup_01}}
 {id=WavePlayer_10, properties {soundWaveAssetPtr=&/Carmine_Dive_01_opt}}
 {id=WavePlayer_11, properties {soundWaveAssetPtr=&/Carmine_Dive_01_distant_opt}}
 {id=WavePlayer_2, properties {soundWaveAssetPtr=&/Carmine_Dive_01}}
 {id=WavePlayer_4, properties {soundWaveAssetPtr=&/Carmine_Dive_01_distant}}
 {id=WavePlayer_5, properties {soundWaveAssetPtr=&/Carmine_GroundPound_Powerup_01_distant}}
 {id=WavePlayer_8, properties {soundWaveAssetPtr=&/Carmine_GroundPound_Powerup_01_opt}}
 {
  id=WavePlayer_9
  properties {soundWaveAssetPtr=&/Carmine_GroundPound_Powerup_01_distant_opt}
 }
]
properties {bOverrideConcurrency=true, maxDistance=25000, volumeMultiplier=1, concurrencyOverrides {maxCount=1, resolutionRule=StopOldest}}
