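# Looping ambient sound cue for the Carmine item, classed as a loot ambient loop.
# A single looping wave player feeds Carmine_ItemLoop_01 straight to the output, with a slight volume boost.
# Inline attenuation overrides give logarithmic distance falloff, stereo spread and listener focus that lowers volume when out of focus.
# The shared Loot_Ambient_Conc concurrency rule governs simultaneous instances.
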SoundCue /Figment_Martin/Sounds/Fort_GamePlay_Sounds/Carmine/Carmine_ItemLoop_Cue
root=WavePlayer_0
nodes [{id=WavePlayer_0, properties {bLooping=true, soundWaveAssetPtr=&/Carmine_ItemLoop_01}}]
properties {
 bOverrideAttenuation=true
 maxDistance=1681.9642
 soundClassObject=/Game/Sounds/FortSoundClassesAndModes/Fort_Loot_Ambient_Loop
 volumeMultiplier=1.3
 attenuationOverrides {
  bEnableListenerFocus=true
  distanceAlgorithm=Logarithmic
  falloffDistance=681.9642
  nonFocusVolumeAttenuation=0.4762
  stereoSpread=35.6359
  attenuationShapeExtents {x=1000}
 }
 concurrencySet [/Game/Sounds/Concurrency_Rules/Loot_Ambient_Conc]
}
