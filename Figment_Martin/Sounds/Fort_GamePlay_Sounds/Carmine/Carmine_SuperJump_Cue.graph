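# Sound cue for the Carmine super jump, blending close and distant recordings, each with its own close or far attenuation asset.
# A quality-level root sends its first two levels to a mixer of standard waves and the other two to a mixer of "opt" variants.
# Every branch overrides its sound class to third-person ranged weapon fire.
# Concurrency allows one instance per owner, stopping the oldest.
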SoundCue /Figment_Martin/Sounds/Fort_GamePlay_Sounds/Carmine/Carmine_SuperJump_Cue
root=QualityLevel_0
nodes [
 {
  id=Attenuation_0
  children [WavePlayer_4]
  properties {attenuationSettings=/Game/Sounds/Attenuation/Fort_Weapon_Carmine_Close_Athena_Att, attenuationOverrides {bSpatialize=false}}
 }
 {
  id=Attenuation_1
  children [WavePlayer_0]
  properties {attenuationSettings=&Fort_Weapon_Carmine_Close_Athena_Att, attenuationOverrides {bSpatialize=false}}
 }
 {
  id=Attenuation_2
  children [WavePlayer_1]
  properties {attenuationSettings=/Game/Sounds/Attenuation/Fort_Weapon_Carmine_Far_Athena_Att, attenuationOverrides {bSpatialize=false}}
 }
 {
  id=Attenuation_3
  children [WavePlayer_5]
  properties {attenuationSettings=&Fort_Weapon_Carmine_Far_Athena_Att, attenuationOverrides {bSpatialize=false}}
 }
 {id=Mixer_0, children [SoundClass_0,SoundClass_1], properties {inputVolume [1,1]}}
 {id=Mixer_1, children [SoundClass_2,SoundClass_3], properties {inputVolume [1,1]}}
 {id=QualityLevel_0, children [Mixer_0,Mixer_0,Mixer_1,Mixer_1]}
 {
  id=SoundClass_0
  children [Attenuation_1]
  properties {soundClassOverride=&Fort_WeaponRanged_Fire_3P_PSM}
 }
 {
  id=SoundClass_1
  children [Attenuation_2]
  properties {soundClassOverride=&Fort_WeaponRanged_Fire_3P_PSM}
 }
 {
  id=SoundClass_2
  children [Attenuation_0]
  properties {soundClassOverride=&Fort_WeaponRanged_Fire_3P_PSM}
 }
 {
  id=SoundClass_3
  children [Attenuation_3]
  properties {soundClassOverride=&Fort_WeaponRanged_Fire_3P_PSM}
 }
 {id=WavePlayer_0, properties {soundWaveAssetPtr=&/Carmine_SuperJump_01}}
 {id=WavePlayer_1, properties {soundWaveAssetPtr=&/Carmine_SuperJump_01_distant}}
 {id=WavePlayer_4, properties {soundWaveAssetPtr=&/Carmine_SuperJump_01_opt}}
 {id=WavePlayer_5, properties {soundWaveAssetPtr=&/Carmine_SuperJump_01_distant_opt}}
]
properties {
 bOverrideConcurrency=true
 maxDistance=25000
 volumeMultiplier=1
 concurrencyOverrides {bLimitToOwner=true, maxCount=1, resolutionRule=StopOldest}
}
