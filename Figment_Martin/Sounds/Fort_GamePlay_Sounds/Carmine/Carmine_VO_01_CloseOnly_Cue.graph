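# Carmine transform sound cue, the close-only layer, playing a single wave through the Carmine close attenuation and a third-person ranged weapon fire sound class.
# A quality-level switch picks the standard wave for the first two levels and an "_opt" variant for the other two, each with its own attenuation and class nodes.
# Mixers retain empty input slots, and the root boosts the live branch's volume.
# Concurrency override allows one instance and blocks new plays.
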SoundCue /Figment_Martin/Sounds/Fort_GamePlay_Sounds/Carmine/Carmine_VO_01_CloseOnly_Cue
root=Mixer_0
nodes [
 {id=Attenuation_2, children [WavePlayer_0], properties {attenuationSettings=/Game/Sounds/Attenuation/Fort_Weapon_Carmine_Close_Athena_Att}}
 {id=Attenuation_4, children [WavePlayer_4], properties {attenuationSettings=&Fort_Weapon_Carmine_Close_Athena_Att}}
 {id=Mixer_0, children [null, QualityLevel_0], properties {inputVolume [0.9,2]}}
 {id=Mixer_1, children [SoundClass_2,null], properties {inputVolume [1,1]}}
 {id=Mixer_2, children [SoundClass_4,null], properties {inputVolume [1,1]}}
 {id=QualityLevel_0, children [Mixer_1,Mixer_1,Mixer_2,Mixer_2]}
 {
  id=SoundClass_2
  children [Attenuation_2]
  properties {soundClassOverride=&Fort_WeaponRanged_Fire_3P_PSM}
 }
 {
  id=SoundClass_4
  children [Attenuation_4]
  properties {soundClassOverride=&Fort_WeaponRanged_Fire_3P_PSM}
 }
 {id=WavePlayer_0, properties {soundWaveAssetPtr=&/Carmine_Transform_01}}
 {id=WavePlayer_4, properties {soundWaveAssetPtr=&/Carmine_Transform_01_opt}}
]
properties {
 bOverrideAttenuation=true
 bOverrideConcurrency=true
 maxDistance=1000000
 volumeMultiplier=1
 attenuationOverrides {falloffDistance=1000000, attenuationShapeExtents {x=0}}
 concurrencyOverrides {maxCount=1, resolutionRule=PreventNew}
}
