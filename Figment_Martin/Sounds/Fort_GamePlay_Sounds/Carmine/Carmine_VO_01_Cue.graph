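# Sound cue for Carmine, named as voice-over, that plays the Ashton_NewFX_Spawn wave at reduced volume in the important in-game mission sound class.
# A Mixer root feeds from that single wave player, and its other input is left empty.
# An attenuation override with an extremely long falloff keeps it audible over huge distances.
# A concurrency override allows one instance and blocks new plays while it runs.
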SoundCue /Figment_Martin/Sounds/Fort_GamePlay_Sounds/Carmine/Carmine_VO_01_Cue
root=Mixer_0
nodes [
 {id=Mixer_0, children [null,WavePlayer_0], properties {inputVolume [0.9,1]}}
 {id=WavePlayer_0, properties {soundWaveAssetPtr=/Game/Sounds/Fort_GamePlay_Sounds/Carmine/Ashton_NewFX_Spawn}}
]
properties {
 bOverrideAttenuation=true
 bOverrideConcurrency=true
 maxDistance=1000000
 soundClassObject=&Fort_Mission_Ingame_Important_PSM
 volumeMultiplier=0.5
 attenuationOverrides {falloffDistance=1000000, attenuationShapeExtents {x=0}}
 concurrencyOverrides {maxCount=1, resolutionRule=PreventNew}
}
