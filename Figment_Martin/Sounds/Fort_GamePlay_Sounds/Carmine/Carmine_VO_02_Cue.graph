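# Sound cue Carmine_VO_02 in the Figment_Martin gameplay sounds, which plays the FrogParty_NewFX_Death wave despite its voice-over name.
# A single wave player is the root, with no randomization, mixing or looping nodes.
# The cue plays at reduced volume through the important in-game mission sound class.
# Overridden attenuation uses a zero inner radius and a huge falloff distance, so it carries extremely far.
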SoundCue /Figment_Martin/Sounds/Fort_GamePlay_Sounds/Carmine/Carmine_VO_02_Cue
root=WavePlayer_1
nodes [{id=WavePlayer_1, properties {soundWaveAssetPtr=&/FrogParty_NewFX_Death}}]
properties {
 bOverrideAttenuation=true
 maxDistance=1000000
 soundClassObject=&Fort_Mission_Ingame_Important_PSM
 volumeMultiplier=0.5
 attenuationOverrides {falloffDistance=1000000, attenuationShapeExtents {x=0}}
}
